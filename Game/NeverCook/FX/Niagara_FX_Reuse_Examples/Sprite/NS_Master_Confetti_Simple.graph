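# Niagara master example system that emits confetti sprites through a single Confetti emitter at a steady spawn rate.
# Particles spawn in a sphere around a sampled skeleton position, with random velocity rotated by the sampled bone rotation, though no skeleton-sampling module appears among the spawn steps.
# Gravity pulls them down while SubUV animation plays and a curve grows then shrinks sprite size.
# Front-end user bools accompany a gameplay/front-end offset modifier, both offsets zero.
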NiagaraSystem /Game/NeverCook/FX/Niagara_FX_Reuse_Examples/Sprite/NS_Master_Confetti_Simple
dependencies {usedDataInterfaces [&NiagaraDataInterfaceCurve], usedDynamicInputs [], usedModules [&SystemState]}
emitters {Confetti {
  emitterUpdateScript [
   &EmitterState(Life Cycle Mode=::NewEnumerator1, Loop Duration=1, UseLoopDelay=false)
   &SpawnRate(SpawnRate=10, Use Spawn Probability=false)
  ]
  particleSpawnScript [
   &InitializeParticle(
    Sprite Size Mode=::NewEnumerator1
    Uniform Sprite Size Max=30
    Uniform Sprite Size Min=20
    UsePositionOffset=false
    Write Material Random=true
   )
   /Niagara/Modules/Spawn/Location/SphereLocation(
    MaskSpawnGroup=false
    Module.Override Randomness=false
    Offset=&FN_FrontEndVector3Modifier(Gameplay=<0 0 0>, Front End=<0 0 0>)
    Sphere Origin:Vector3f=$Particles.SampleSkeletalMeshSkeleton.SampledPosition
    Sphere Radius=&UniformRangedFloat(Minimum=15, Maximum=30)
    Surface Only=false
   )
   &AddVelocity(Velocity=&RotateVectorByQuaternion(
     Vector To Rotate=&UniformRangedVector(Minimum=<-1 -1 -1>, Maximum=<1 1 1>)
     Quaternion=$Particles.SampleSkeletalMeshSkeleton.SampledRotation
    ))
   &ApplyInitialForces()
   /SetVariables() {isSetParameters=true}
  ]
  particleUpdateScript [
   &ParticleState()
   &ScaleSpriteSize(Scale Factor=&Vector2DFromFloat(Value=&FloatFromCurve(FloatCurve=DataInterface() {propertyValues {Curve {keys [
          {arriveTangent=2.3096, interpMode=RCIM_Cubic, leaveTangent=2.3096, tangentMode=RCTM_User, time=0, value=0}
          {arriveTangent=-0.0952, interpMode=RCIM_Cubic, leaveTangent=-0.0952, tangentMode=RCTM_User, time=0.431, value=0.3973}
          <at 0.9969 0.0123 RCIM_Cubic>
         ]}}})))
   /Niagara/Modules/Update/SubUV/SubUVAnimation(Number Of Frames=36)
   &GravityForce(Gravity=<0 0 -150>)
   &SolveForcesAndVelocity(Clamp Velocity=false, Limit Acceleration=false)
  ]
  rendererClasses [&NiagaraSpriteRendererProperties]
  renderers [{rendererClass=&NiagaraSpriteRendererProperties}]
 }}
summary {
 systemName=NS_Master_Confetti_Simple
 emitters [{emitterName=Confetti, rendererClasses [&NiagaraSpriteRendererProperties]}]
 userVariables [
  User.bIsFrontEnd {type=NiagaraBool}
  User.bIsFrontEndPreview {type=NiagaraBool}
 ]
}
