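# Niagara example system in the FX reuse examples folder: a single sprite emitter that continuously spawns short-lived, randomly sized flare sprites.
# Color cycles through a rainbow curve over particle life at half intensity, while alpha fades quickly in and out.
# The alpha curve is scaled by emitter scalability distance fraction, fading flares at the far end of their range.
# A Material module feeds emitter age into distortion texture time alongside addition and power settings.
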NiagaraSystem /Game/NeverCook/FX/Niagara_FX_Reuse_Examples/Sprite/NS_Sprite_Flare
dependencies {
 usedDataInterfaces [&NiagaraDataInterfaceColorCurve, &NiagaraDataInterfaceCurve]
 usedDynamicInputs [&SimulationPosition]
 usedModules [&SystemState]
}
emitters {2_Flare001 {
  emitterUpdateScript [
   &EmitterState(Life Cycle Mode=::NewEnumerator1, Loop Duration=1, UseLoopDelay=false)
   &SpawnRate(SpawnRate=10, Use Spawn Probability=false)
  ]
  particleSpawnScript [&InitializeParticle(
    Color=<rgba 1 0 0 1>
    Lifetime=&UniformRangedFloat(Minimum=0.4, Maximum=0.5)
    Mass Mode=::NewEnumerator1
    Sprite Size=&Multiply_Vector2(A=&UniformRangedVector2D(Minimum=<10 25>, Maximum=<15 50>), B=<1 1>)
    Sprite Size Mode=::NewEnumerator4
    UsePositionOffset=false
    Write Material Random=true
   ) {module=InitializeParticle001}]
  particleUpdateScript [
   &ParticleState()
   &Material(Addition=0.0727, Distortion Texture Time:NiagaraFloat=$Emitter.Age, Power=1.5416)
   &Color(
    Color=&ColorFromCurve(ColorCurve=DataInterface() {propertyValues {
       AlphaCurve {keys [<at 0 1>,<at 1 1>]}
       BlueCurve {keys [<at 0.011 1>,<at 0.1666 1>,<at 0.323 0>,<at 0.5 0>,<at 0.6875 0>,<at 0.8271 1>,<at 0.989 1>]}
       GreenCurve {keys [<at 0.011 0>,<at 0.1666 0>,<at 0.323 0>,<at 0.5 1>,<at 0.6875 1>,<at 0.8271 1>,<at 0.989 0>]}
       RedCurve {keys [<at 0.011 0>,<at 0.1666 1>,<at 0.323 1>,<at 0.5 1>,<at 0.6875 0>,<at 0.8271 0>,<at 0.989 0>]}
      }})
    Scale Alpha=&FloatFromCurve(
     FloatCurve=DataInterface() {propertyValues {Curve {keys [<at -0.0045 0.0315>,<at 0.0491 0.9978>,<at 0.942 0.9978>,<at 1.0011 0.002>]}}}
     Scale Curve=&FloatFromCurve(FloatCurve=DataInterface() {propertyValues {Curve {keys [<at 0.5 1>,<at 1.0006 0>]}}}, CurveIndex=$Emitter.Scalability.DistanceFraction)
    )
    Scale Color=<0.5 0.5 0.5>
   )
  ]
  rendererClasses [&NiagaraSpriteRendererProperties]
  renderers [{rendererClass=&NiagaraSpriteRendererProperties}]
 }}
summary {
 systemName=NS_Sprite_Flare
 emitters [{emitterName=2_Flare001, rendererClasses [&NiagaraSpriteRendererProperties]}]
 userVariables [
  User.bIsFrontEnd {type=NiagaraBool}
  User.bIsFrontEndPreview {type=NiagaraBool}
 ]
}
